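# Sound cue for activating Wasteland challenges, playing one of three "Challenge Accepted" waves.
# An equally weighted Random node picks the wave, with a Modulator giving the second take a fixed volume boost at unchanged pitch.
# It uses its own attenuation asset and the Fort_Quests sound class.
# A concurrency override allows one instance per owner, stopping the oldest when a new one starts.
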SoundCue /ChallengeQuests/Audio/Cues/SC_Wasteland_Challenges_Activate
root=Random_0
nodes [
 {id=Modulator_0, children [WavePlayer_2], properties {pitchMax=1, pitchMin=1, volumeMax=1.4, volumeMin=1.4}}
 {id=Random_0, children [WavePlayer_0,Modulator_0,WavePlayer_3], properties {bShouldExcludeFromBranchCulling=true, weights [1,1,1]}}
 {id=WavePlayer_0, properties {soundWaveAssetPtr=/ChallengeQuests/Audio/Waves/Wasteland_Challenge_Accepted_01}}
 {id=WavePlayer_2, properties {soundWaveAssetPtr=/ChallengeQuests/Audio/Waves/Wasteland_Challenge_Accepted_02}}
 {id=WavePlayer_3, properties {soundWaveAssetPtr=/ChallengeQuests/Audio/Waves/Wasteland_Challenge_Accepted_03}}
]
properties {
 attenuationSettings=/ChallengeQuests/Audio/Settings/Att_Wasteland_Challenges_Activate
 bOverrideConcurrency=true
 maxDistance=30000
 soundClassObject=&Fort_Quests
 volumeMultiplier=1
 concurrencyOverrides {bLimitToOwner=true, maxCount=1, resolutionRule=StopOldest, volumeScaleAttackTime=0.1}
}
