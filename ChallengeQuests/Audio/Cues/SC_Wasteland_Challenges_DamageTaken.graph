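# Sound cue for the damage-taken hit notification in Wasteland challenge quests, routed to the received 2D hit notification sound class.
# A Branch on the IsLocal parameter plays audio only when true, leaving the false and unset inputs silent.
# That path runs through a Modulator into an equally weighted Random node, excluded from branch culling, that picks one of six DamageTaken waves.
# The cue plays at reduced volume with dedicated hit-marker attenuation settings.
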SoundCue /ChallengeQuests/Audio/Cues/SC_Wasteland_Challenges_DamageTaken
root=Branch_0
nodes [
 {id=Branch_0, children [Modulator_0,null,null], properties {boolParameterName=IsLocal}}
 {id=Modulator_0, children [Random_0]}
 {
  id=Random_0
  children [WavePlayer_0,WavePlayer_1,WavePlayer_2,WavePlayer_3,WavePlayer_4,WavePlayer_5]
  properties {bShouldExcludeFromBranchCulling=true, weights [1,1,1,1,1,1]}
 }
 {id=WavePlayer_0, properties {soundWaveAssetPtr=/ChallengeQuests/Audio/Waves/Wasteland_Challenges_DamageTaken_06}}
 {id=WavePlayer_1, properties {soundWaveAssetPtr=/ChallengeQuests/Audio/Waves/Wasteland_Challenges_DamageTaken_05}}
 {id=WavePlayer_2, properties {soundWaveAssetPtr=/ChallengeQuests/Audio/Waves/Wasteland_Challenges_DamageTaken_04}}
 {id=WavePlayer_3, properties {soundWaveAssetPtr=/ChallengeQuests/Audio/Waves/Wasteland_Challenges_DamageTaken_03}}
 {id=WavePlayer_4, properties {soundWaveAssetPtr=/ChallengeQuests/Audio/Waves/Wasteland_Challenges_DamageTaken_02}}
 {id=WavePlayer_5, properties {soundWaveAssetPtr=/ChallengeQuests/Audio/Waves/Wasteland_Challenges_DamageTaken_01}}
]
properties {
 attenuationSettings=/ChallengeQuests/Audio/Settings/Att_WastelandChallenges_HitMarker_Receive
 maxDistance=2097152
 soundClassObject=/Game/Sounds/FortSoundClassesAndModes/Fort_2DHitNotifications_Receive_PSM
 volumeMultiplier=0.5
}
